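; Weapon Blueprint for the MurkyVole gauntlet guns, dual-hand guns that overheat as they fire.
; Each fire event swaps HandFiring, whose OnRep plays that hand's FX on the cached gauntlet character part, which also supplies muzzle sockets.
; A timer smooths heat into loop audio and a faked ammo value for low-ammo sounds.
; On the owning client, another timer checks aimed-at enemies with line of sight to update missile target state.
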(event EventBeginPlay
  (bind _returnvalue (Game|GetInstigator))
  ;[1 1 1] On begin play, listed for weapons fire event and swap last hand fired each time
  (Ability|Async|WaitGameplayEventToActor _returnvalue "(TagName=\"Event.Weapons.Fire\")"
    (:then
      (Default|GetandUpdateCCPMGauntletsRef)
      (Variables|Default|SetbIsLocal (Pawn|IsLocallyViewed _returnvalue)) ; Used for low ammo sound setup
      (bind _asfort_player_controller (Utilities|Casting|CastToFortPlayerController (Fortnite|GetControllerfromActor (Game|GetInstigator))))
      (Variables|Default|SetCachedPlayerController _asfort_player_controller)
      (Variables|MissileTarget|SetMissileCheckTargetFrequency (ScalableFloat|GetValueatLevel (Variables|MissileTarget|GetRow_MissileCheckUIFrequency))))
    (:EventReceived
      (switch Utilities|FlowControl|Switch|SwitchonEDualWeaponHand (Variables|Default|GetHandFiring)
        (:LEFT
          (Variables|Default|SetHandFiring "RIGHT"))
        (:RIGHT
          (Variables|Default|SetHandFiring "LEFT"))))))

(event FortWeapon|EventOnEquippedAndAttached
  ;[1 1 1] OnEquip - Reset MuzzleGlowMID Opacity
  (|CallParentFunction)
  (Networking|CanExecuteCosmeticEvents
    (:True
      (bind _instigatorpawn (CallFunction|GetInstigatorPawn))
      (Customization|AssignOnFinishedCharacterCustomization _instigatorpawn (delegate Custom|OnFinishedCharacterCustomization_Event) :!pinned true) ; TODO: Veify this works on remote clients when entering relevancy
      (Default|GetandUpdateCCPMGauntletsRef)
      (Audio|Parameter|SetFloatParameter (Variables|Default|GetOverheatLoopSound) "Overheat" 0.0 :!class /Script/AudioExtensions.AudioParameterControllerInterface)
      (CallFunction|StartMissileTargetCheck))))

(event Custom|OnFinishedCharacterCustomization_Event
  (params (Pawn :type Fort%20Player%20Pawn%20Object%20Reference))
  ;[1 1 1] OnEquip - Reset MuzzleGlowMID Opacity
  (Default|GetandUpdateCCPMGauntletsRef)
  (Development|PrintString "Finished Customization - Update CCPM" true true "(R=0.091180,G=1.000000,B=0.000000,A=1.000000)" 5.0))

(event EventOnPlayWeaponFireFX
  (params (bPersistentFire :type Boolean) (bSecondaryFire :type Boolean))
  ;[1 1 1] OnPlayWeaponFireFX
  (|CallParentFunction bPersistentFire bSecondaryFire)
  (Development|PrintString "Play Fire FX")
  (Audio|Components|Audio|Play (Variables|Default|GetOverheatLoopSound) 0.0)
  (FortWeapon|PlayForceFeedback false (Effects|ShouldPlayPersistentFireFX))
  (Variables|Default|SetbJustStartedOverheat true)
  (bind _returnvalue (Utilities|Time|SetTimerbyEvent (EventDispatchers|CreateEvent :!event Custom|BarrelOverheatFXandSound) 0.0333 true true))
  (Variables|Default|SetGlowTimer _returnvalue)
  (Utilities|FlowControl|RetriggerableDelay 0.2)
  (Variables|Default|SetbJustStartedOverheat false))

(event EventOnStopWeaponFireFX
  ;[1 1 1] On Stop Weapon FX
  (|CallParentFunction)
  (FortWeapon|StopForceFeedback)
  (Development|PrintString "Stop Fire FX"))

(event Overheating|EventOnStartOverheated
  (bind _ccpm_gauntlets (Variables|Default|GetCCPM_Gauntlets))
  (bind _returnvalue (CustomCharacter|GetSkeletalMeshForPartOnPlayerPawn _ccpm_gauntlets))
  ;[1 1 1] Overheat Started
  (|CallParentFunction)
  (Development|PrintString "OVERHEATING")
  (Utilities|FlowControl|Sequence
    (:then_0
      (Default|GetandUpdateCCPMGauntletsRef)
      (Utilities|IsValid _ccpm_gauntlets
        (:Is%20Valid
          (if (Utilities|IsValid _returnvalue :!class /Script/Engine.KismetSystemLibrary)
            (Niagara|SpawnSystemAttached "/MurkyVoleGameplay/FX/Systems/NS_MurkyVole_Overheat_Gauntlet.NS_MurkyVole_Overheat_Gauntlet" _returnvalue "None" "0, 0, 0" "0, 0, 0" "KeepRelativeOffset" true)))))
    (:then_1
      (Audio|SpawnSoundAttached (Variables|Default|GetOverheatSound) (Variables|Visual|GetWeaponMesh))
      (bind _returnvalue_2 (Utilities|Time|SetTimerbyEvent (EventDispatchers|CreateEvent :!event Custom|PlayOverheatForceFeedback) 0.1 :!pinned true)) ; HACK - to prevent this getting cancelled by Stop Weapon Fire FX
      (Variables|Default|SetTimerHandle_PlayOverheatForceFeedback _returnvalue_2))))

(event Custom|PlayOverheatForceFeedback
  ;[1 1 1] Play Overheat Force Feedback
  (FortWeapon|PlayForceFeedback true))

(event Custom|BarrelOverheatFXandSound
  (bind _overheatsmoothed (Variables|Default|GetOverheatSmoothed))
  (bind _returnvalue (Math|Float|MapRangeClamped _overheatsmoothed 0.08 1.0 0.0 1.0))
  ;[1 1 1] BarrelGlow - Handles ramping overheat FX and sound.  Timer shuts off when there is no heat in the minigun
  (if (or (> (Default|CurrentScalarValue) 0.0) (Variables|Default|GetbJustStartedOverheat))
    (Networking|CanExecuteCosmeticEvents
      (:True
        (Variables|Default|SetOverheatSmoothed (Math|Interpolation|FInterpTo (Variables|Default|GetOverheatSmoothed) (Math|Float|Lerp 0.0 1.0 (Default|CurrentScalarValue)) (Utilities|Time|GetWorldDeltaSeconds) 1e+01))
        (Audio|Parameter|SetFloatParameter (Variables|Default|GetOverheatLoopSound) "Overheat" _returnvalue :!class /Script/AudioExtensions.AudioParameterControllerInterface)
        (CallFunction|SetLowAmmoAudio _returnvalue)))
    (else
      (Utilities|Time|ClearandInvalidateTimerbyHandle (Variables|Default|GetGlowTimer))
      (if (Audio|Components|Audio|IsPlaying (Variables|Default|GetOverheatLoopSound))
        (Audio|Components|Audio|FadeOut (Variables|Default|GetOverheatLoopSound) 0.1 0.0 "Logarithmic" :!class /Script/Engine.AudioComponent)))))

(event FortWeapon|EventOnLocalUnEquip
  ;[1 1 1] Clear Timer Reset MuzzleGlowMID Opacity/Sound.  Keeps a phantom sound from playing.
  (|CallParentFunction)
  (Utilities|Time|ClearandInvalidateTimerbyHandle (Variables|Default|GetGlowTimer))
  (Networking|CanExecuteCosmeticEvents
    (:True
      (Audio|Parameter|SetFloatParameter (Variables|Default|GetOverheatLoopSound) "Overheat" 0.0 :!class /Script/AudioExtensions.AudioParameterControllerInterface))))

(event Custom|SetLowAmmoAudio
  (params (Overheat :type Float%20%28double-precision%29))
  (bind _returnvalue (AudioParameter|GetAudioParameterComponent))
  (bind _lowammosoundthreshold (Variables|Default|GetLowAmmoSoundThreshold))
  ;[1 1 1] Sending a correct ammo value to play low ammo sounds due to the weapon having a constant local magazine ammo count
  (if (Variables|Default|GetbIsLocal)
    (if (> Overheat _lowammosoundthreshold)
      (Utilities|IsValid _returnvalue
        (:Is%20Valid
          (Audio|Parameter|SetIntegerParameter _returnvalue "Weapon.AmmoRemaining" (Math|Float|Truncate (Math|Float|MapRangeClamped Overheat 1.0 _lowammosoundthreshold 0.0 2e+01)) :!class /Script/AudioExtensions.AudioParameterControllerInterface))))))

(event FortWeapon|EventOnUnEquippedAndDetached
  ;[1 1 1] On Unequipped and Detached
  (|CallParentFunction)
  (Utilities|Time|ClearandInvalidateTimerbyHandle (Variables|Default|GetTimerHandle_PlayOverheatForceFeedback))
  (FortWeapon|StopForceFeedback)
  (Utilities|Time|ClearandInvalidateTimerbyHandle (Variables|MissileTarget|GetMissileCheckTargetTimerHandle))
  (bind _instigatorpawn (CallFunction|GetInstigatorPawn))
  (Customization|UnbindEventfromOnFinishedCharacterCustomization _instigatorpawn (EventDispatchers|CreateEvent :!event Custom|OnFinishedCharacterCustomization_Event)))

(fn CurrentScalarValue
  (return (/ (Overheat|GetCurrentOverheatValue) (Overheat|GetOverheatingMaxValue))))

(fn GetAimedatEnemywithLOS
  (CallFunction|GetInstigatorPawn)
  (bind _returnvalue (Fortnite|GetAimedAtEnemy (Variables|Default|GetCachedPlayerController) (ScalableFloat|GetValueatLevel (Variables|MissileTarget|GetRow_PawnLock_HalfWidth)) (ScalableFloat|GetValueatLevel (Variables|MissileTarget|GetRow_PawnLock_HalfHeight)) (ScalableFloat|GetValueatLevel (Variables|MissileTarget|GetRow_PawnLock_Range)) true))
  (bind _output_get (LocalVariables|SetTargetEnemy _returnvalue))
  (Utilities|IsValid _output_get
    (:Is%20Valid
      (bind _returnvalue_2 (Fortnite|CheckLineOfSightToActorWithChannel (Pawn|GetPlayerViewPoint (Variables|Default|GetCachedPlayerController)) (LocalVariables|GetTargetEnemy) "ECC_GameTraceChannel7" (Variables|Default|GetCachedPlayerPawn)))
      (if _returnvalue_2
        (return (LocalVariables|GetTargetEnemy))
        (else
          _)))
    (:Is%20Not%20Valid
      _))
  (return 0))

(fn GetAndUpdateCCPMGauntletsRef
  (bind _ccpm_gauntlets (Variables|Default|GetCCPM_Gauntlets)
    (:then
      (return _ccpm_gauntlets))
    (:else
      (bind _instigatorpawn (CallFunction|GetInstigatorPawn))
      (bind _asccpm_athena_murky_vole_gauntlets (Utilities|Casting|CastToCCPM_Athena_MurkyVole_Gauntlets (FortPawn|GetModifierforPartType _instigatorpawn "Gameplay")))
      (bind _output_get (Variables|Default|SetCCPM_Gauntlets _asccpm_athena_murky_vole_gauntlets))
      (return _output_get))))

(fn GetInstigatorPawn
  (Utilities|IsValid (Variables|Default|GetCachedPlayerPawn)
    (:Is%20Valid
      _)
    (:Is%20Not%20Valid
      (bind _asfort_player_pawn (Utilities|Casting|CastToFortPlayerPawn (Game|GetInstigator)))
      (Variables|Default|SetCachedPlayerPawn _asfort_player_pawn)))
  (return (Variables|Default|GetCachedPlayerPawn)))

(fn GetMuzzleLocation
  (params (PatternIndex :type Integer))
  (bind (_meshcomponent _socketlocation) (Default|GetSocketLocationandRotation (Variables|Default|GetCCPM_Gauntlets) (Utilities|Select :Index (Variables|Default|GetHandFiring) :LEFT "Muzzle_L" :RIGHT "Muzzle_R")))
  (Variables|Default|GetCCPM_Gauntlets
    (:then
      (return _socketlocation))
    (:else
      (return (Transformation|GetSocketLocation (Variables|/Script/Engine.Character|GetMesh (Variables|Default|GetCachedPlayerPawn)) (Utilities|Select :Index (Variables|Default|GetHandFiring) :LEFT "Hand_L" :RIGHT "Hand_R"))))))

(fn OnRep_HandFiring
  (bind _ccpm_gauntlets (Default|GetandUpdateCCPMGauntletsRef))
  (Class|CCPMAthenaMurkyVoleGauntlets|PlayFireFXFromHand _ccpm_gauntlets (Variables|Default|GetHandFiring)))

(fn StartMissileTargetCheck
  (CallFunction|GetInstigatorPawn)
  (if (and (Utilities|Enum|Equal(Enum) (Networking|GetLocalRole (Variables|Default|GetCachedPlayerPawn)) "ROLE_AutonomousProxy") (> (Variables|MissileTarget|GetMissileCheckTargetFrequency) 0))
    (bind _returnvalue (Utilities|Time|SetTimerbyEvent (EventDispatchers|CreateEvent :!event UpdateMissileCurrentTarget) (Variables|MissileTarget|GetMissileCheckTargetFrequency) true true))
    (Variables|MissileTarget|SetMissileCheckTargetTimerHandle _returnvalue)))

(fn UpdateMissileCurrentTarget
  (bind _foundenemy (CallFunction|GetAimedAtEnemyWithLOS))
  (Class|FortWeaponComponentTargetState|UpdateTargetInfo :self (Variables|Default|GetFortWeaponComponent_TargetState) :bIsTargetValid (Utilities|IsValid _foundenemy :!class /Script/Engine.KismetSystemLibrary)))

(fn ConstructionScript
  (|CallParentFunction))
