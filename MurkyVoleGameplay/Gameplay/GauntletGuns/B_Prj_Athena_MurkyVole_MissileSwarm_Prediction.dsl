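; Predicted homing missile projectile for the MurkyVole missile swarm, with cosmetic mesh scale-up, rocket spin, trail and audio.
; Reaching or passing the homing target tries to explode without allowing it on server, and non-servers stop simulating.
; A lifetime timer or authority-side touch also stops simulation, and stopping tries to explode with server allowed.
; Explosion or destruction halts movement, hides the mesh, kills the trail and fades audio.
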(event Custom|BeginScaleMissileMesh
  (bind _returnvalue_x (Math|Vector|BreakVector (Transformation|GetWorldScale (Variables|Default|GetSM_OstrichRocket))))
  ;[1 1 1] Begin Scale Missile Mesh
  (Networking|CanExecuteCosmeticEvents
    (:True
      (Variables|Default|SetMissileScale_Start _returnvalue_x)
      (bind (_direction _missilescalemultiplier) (Timeline|ScaleUpMesh PlayFromStart)
        (:Update
          (Transformation|SetWorldScale3D (Variables|Default|GetSM_OstrichRocket) (Math|Conversions|ToVector(Float) (* _missilescalemultiplier (Variables|Default|GetMissileScale_Start)))))
        (:Finished
          (Transformation|SetWorldScale3D (Variables|Default|GetSM_OstrichRocket) (Math|Conversions|ToVector(Float) (Variables|Default|GetMissileScale_Start))))))))

(event EventBeginPlay
  ;[1 1 1] Begin Play
  (Utilities|Time|SetTimerbyFunctionName self "FuseTimerMax" (ScalableFloat|GetValueatLevel (Variables|Balance|GetRow_ProjectileMaxLifetime)))
  (Networking|CanExecuteCosmeticEvents
    (:True
      (CallFunction|BeginScaleMissileMesh))))

(event EventTick
  (params (DeltaSeconds :type Float%20%28single-precision%29))
  ;[1 1 1] On Tick, spin the rocket and particle trail
  (Variables|Default|SetTick%20Delta DeltaSeconds)
  (Networking|CanExecuteCosmeticEvents
    (:True
      (if (Components|Activation|IsActive (Variables|Default|GetNS_MurkyVole_PrjTrail))
        (Variables|Effects|SetRocketSpin (Math|Rotator|MakeRotator 0.0 (* (Variables|Default|GetTick%20Delta) (Variables|Effects|GetRocketSpinRate))))
        (Transformation|AddLocalRotation (Variables|Default|GetSM_OstrichRocket) (Variables|Effects|GetRocketSpin) false false)))))

(event Custom|FuseTimerMax
  ;[1 1 1] Fuse Timer Max
  (Game|Components|ProjectileMovement|StopSimulating (Variables|Movement|GetProjectileMovementComponent) (Collision|MakeHitResult true false 0.0 0.0 (Transformation|GetActorLocation) (Transformation|GetActorLocation))))

(event ProjectileMovement|EventOnReachedLockTargetDistanceThreshold
  ;[1 1 1] On Reached Lock Target Distance Threshold
  (Development|PrintString "Reached Target Distance Threshold" (Variables|Default|GetDebug) (Variables|Default|GetDebug))
  (CallFunction|TryExplode false)
  (if (not (Networking|IsServer))
    (Game|Components|ProjectileMovement|StopSimulating (Variables|Movement|GetProjectileMovementComponent) (Collision|MakeHitResult true false 0.0 0.0 (Transformation|GetActorLocation) (Transformation|GetActorLocation)))))

(event ProjectileMovement|EventOnHomingPassedTarget
  ;[1 1 1] On Passed Homing Target
  (Development|PrintString "Passed Target" (Variables|Default|GetDebug) (Variables|Default|GetDebug))
  ;[1 1 1] On Reached Lock Target Distance Threshold
  (CallFunction|TryExplode false)
  (if (not (Networking|IsServer))
    (Game|Components|ProjectileMovement|StopSimulating (Variables|Movement|GetProjectileMovementComponent) (Collision|MakeHitResult true false 0.0 0.0 (Transformation|GetActorLocation) (Transformation|GetActorLocation)))))

(event EventOnStop
  (params (Hit :type Hit%20Result%20Structure%20%28by%20ref%29))
  ;[1 1 1] On Stop
  (CallFunction|TryExplode true)
  (Development|PrintString "Stopped"))

(event Custom|TryExplode
  (params (bAllowOnServer :type Boolean))
  ;[1 1 1] Try Explode
  (CallFunction|TrytoExplode bAllowOnServer))

(event EventOnExploded
  (params (HitActors :type Array%20of%20Actor%20Object%20References) (HitResults :type Array%20of%20Hit%20Result%20Structures))
  ;[1 1 1] On Exploded
  (Components|Movement|StopMovementImmediately (Variables|Movement|GetProjectileMovementComponent))
  (Networking|CanExecuteCosmeticEvents
    (:True
      (Development|SetHiddeninGame (Variables|Default|GetSM_OstrichRocket) true false)
      (Components|Activation|SetActive (Variables|Default|GetNS_MurkyVole_PrjTrail) false true)
      (Components|Activation|Deactivate (Variables|Default|GetNS_MurkyVole_PrjTrail))
      (Audio|Components|Audio|FadeOut (Variables|Default|GetRocket_AudioComponent) 0.1 0.0 "Linear" :!class /Script/Engine.AudioComponent))))

(event EventDestroyed
  ;[1 1 1] On Exploded
  (Components|Movement|StopMovementImmediately (Variables|Movement|GetProjectileMovementComponent))
  (Networking|CanExecuteCosmeticEvents
    (:True
      (Development|SetHiddeninGame (Variables|Default|GetSM_OstrichRocket) true false)
      (Components|Activation|SetActive (Variables|Default|GetNS_MurkyVole_PrjTrail) false true)
      (Components|Activation|Deactivate (Variables|Default|GetNS_MurkyVole_PrjTrail))
      (Audio|Components|Audio|FadeOut (Variables|Default|GetRocket_AudioComponent) 0.1 0.0 "Linear" :!class /Script/Engine.AudioComponent))))

(event EventOnTouched
  (params (OtherActor :type Actor%20Object%20Reference) (OtherComp :type Primitive%20Component%20Object%20Reference) (HitResult :type Hit%20Result%20Structure%20%28by%20ref%29) (bIsOverlap :type Boolean))
  (Utilities|Casting|CastToFortWaterBodyActor OtherActor
    (:then
      _)
    (:CastFailed
      (Utilities|Casting|CastToFortAthenaWater OtherActor)))
  (if (Utilities|Enum|Equal(Enum) (Networking|GetLocalRole) "ROLE_Authority")
    (Game|Components|ProjectileMovement|StopSimulating (Variables|Movement|GetProjectileMovementComponent) HitResult)))

(fn ConstructionScript)
