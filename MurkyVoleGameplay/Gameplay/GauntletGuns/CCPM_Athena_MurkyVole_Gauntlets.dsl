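; Character part modifier for the MurkyVole gauntlet guns that manages their muzzle and shell burst effects on the player.
; On pawn and part change, InitAttachments snaps left and right effect components to matching Muzzle and Shells sockets on the part's skeletal mesh.
; PlayFireFXFromHand switches on the fired hand and activates only that side's shell burst, not the muzzle effect.
; A helper returns the mesh and a socket transform; the other event handlers are empty stubs.
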(event Custom|PlayFireFXFromHand
  (params (HandFired :type EDualWeaponHand%20Enum :default "LEFT"))
  ;[1 1 1] Play Fire FX From Hand
  (switch Utilities|FlowControl|Switch|SwitchonEDualWeaponHand HandFired
    (:LEFT
      (Components|Activation|Activate (Variables|Default|GetNS_Shells_Burst_L) false))
    (:RIGHT
      (Components|Activation|Activate (Variables|Default|GetNS_Shells_Burst_R) false))))

(event EventOnPawnAndPartChanged
  ;[1 1 1] On Pawn and Part Changed
  (|CallParentFunction)
  (CallFunction|InitAttachments))

(event EventOnPostResIn)

(event Custom|InitAttachments
  ;[1 1 1] Init Attachments
  (Transformation|AttachComponentToComponent (Variables|Default|GetNS_Muzzle_Burst_R) (CustomCharacter|GetSkeletalMeshForPartOnPlayerPawn) "Muzzle_R" "SnapToTarget" "SnapToTarget" "KeepRelative" true)
  (Transformation|AttachComponentToComponent (Variables|Default|GetNS_Muzzle_Burst_L) (CustomCharacter|GetSkeletalMeshForPartOnPlayerPawn) "Muzzle_L" "SnapToTarget" "SnapToTarget" "KeepRelative" true)
  (Transformation|AttachComponentToComponent (Variables|Default|GetNS_Shells_Burst_R) (CustomCharacter|GetSkeletalMeshForPartOnPlayerPawn) "Shells_R" "SnapToTarget" "SnapToTarget" "KeepRelative" true)
  (Transformation|AttachComponentToComponent (Variables|Default|GetNS_Shells_Burst_L) (CustomCharacter|GetSkeletalMeshForPartOnPlayerPawn) "Shells_L" "SnapToTarget" "SnapToTarget" "KeepRelative" true))

(event EventOnNewPlayerWeaponType
  (params (NewType :type EFortItemType%20Enum :default "WorldItem") (OldType :type EFortItemType%20Enum :default "WorldItem")))

(event EventOnWeaponChanged
  (params (NewWeapon :type Fort%20Weapon%20Object%20Reference) (OldWeapon :type Fort%20Weapon%20Object%20Reference)))

(event EventOnWeaponFired)

(event EventOnNewWeapons
  (params (WeaponsAsItems :type Array%20of%20Fort%20Item%20Object%20References)))

(event EventOnEnterVehicle
  (params (VehicleEntered :type Fort%20Vehicle%20Interface%20Interface%20%28by%20ref%29)))

(event EventOnExitVehicle)

(event EventOnTargetingOrFirstPersonCameraChanged
  (|CallParentFunction) ; This node is disabled and will not be called.\nDrag off pins to build functionality.
)

(fn GetSocketLocationAndRotation
  (params (SocketName :type Name))
  (return (CustomCharacter|GetSkeletalMeshForPartOnPlayerPawn) (Transformation|GetSocketLocation (CustomCharacter|GetSkeletalMeshForPartOnPlayerPawn) (Variables|GetSocketName)) (Transformation|GetSocketRotation (CustomCharacter|GetSkeletalMeshForPartOnPlayerPawn) (Variables|GetSocketName))))

(fn ConstructionScript
  (|CallParentFunction))
